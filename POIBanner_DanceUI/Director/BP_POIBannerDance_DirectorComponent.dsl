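; UI director component that shows and hides the POI banner dance progress scene for the owning local player.
; When UI is ready, it listens on the global Gameplay Event Router for zone access and result-animation-finished events.
; Zone access payloads drive adding or removing the scene, while the animation-finished event always removes it.
; Both listener handles are stored and stopped on EndPlay.
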(event EventOnReadyToShowUI
  (bind (_asynctaskproxy _payload) (GameplayEventRouter|StartListeningToEvent (GameplayEventRouter|GetGlobalEventRouter) "/POIBanner/DataTypes/POIZoneAccessInfo.POIZoneAccessInfo" (CallFunction|GetOwningLocalPlayerController))
    (:then
      (Variables|Default|SetPOIInteractionEvent _asynctaskproxy)
      (bind _asynctaskproxy_2 (GameplayEventRouter|StartListeningToEvent (GameplayEventRouter|GetGlobalEventRouter) "/POIBanner/DataTypes/POIResultAnimFinished.POIResultAnimFinished" (CallFunction|GetOwningLocalPlayerController))
        (:then
          (Variables|Default|SetPOIResultAnimEvent _asynctaskproxy_2))
        (:OnEventReceived
          (CallFunction|RemoveScene (Variables|Default|GetPOIDanceProgressScene)))))
    (:OnEventReceived
      (CallFunction|AddorRemoveScene (Variables|Default|GetPOIDanceProgressScene) (Utilities|Struct|BreakPOIZoneAccessInfo _payload)))))

(event EventEndPlay
  (params (EndPlayReason :type EEndPlayReason%20Enum :default "Destroyed"))
  (GameplayEventRouter|StopListeningtoEvent (Variables|Default|GetPOIInteractionEvent))
  (GameplayEventRouter|StopListeningtoEvent (Variables|Default|GetPOIResultAnimEvent)))
